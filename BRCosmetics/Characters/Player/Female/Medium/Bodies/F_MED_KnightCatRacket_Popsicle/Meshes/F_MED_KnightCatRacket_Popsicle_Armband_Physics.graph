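# Physics asset for the armband on the Popsicle variant of the medium female KnightCatRacket cosmetic body.
# It has capsule bodies for the right lower-arm twist bone, a small base_armring anchor and the larger armring.
# The anchor's angular constraint to the forearm is locked, while the ring's constraint to the anchor has tight swing and twist limits, widest on swing 2, allowing slight wobble on the arm.
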PhysicsAsset /BRCosmetics/Characters/Player/Female/Medium/Bodies/F_MED_KnightCatRacket_Popsicle/Meshes/F_MED_KnightCatRacket_Popsicle_Armband_Physics
bodies {
 armring [{shapeType=Capsule, length=4.0045, radius=3.9079, shapeName=armring_capsule, center=<-0.0003 0 0.0034>, rotation=<rot 27.8673 -90 27.0785>}]
 base_armring [{shapeType=Capsule, length=0.5, radius=0.5, shapeName=base_armring_capsule, rotation=<rot 0 -90 0>}]
 lowerarm_twist_01_r [{
   shapeType=Capsule
   length=7.4858
   radius=2.4375
   shapeName=lowerarm_twist_01_r_capsule
   center=<7.0354 -0.5487 -0.5957>
   rotation=<rot -1.193 87.9151 -87.8916>
  }]
}
constraints [
 {
  bone1Name=base_armring
  bone2Name=lowerarm_twist_01_r
  swing1LimitDegrees=45
  swing1Motion=Locked
  swing2LimitDegrees=45
  swing2Motion=Locked
  twistLimitDegrees=45
  twistMotion=Locked
 }
 {bone1Name=armring, bone2Name=base_armring, swing1LimitDegrees=2, swing2LimitDegrees=20, twistLimitDegrees=3}
]
