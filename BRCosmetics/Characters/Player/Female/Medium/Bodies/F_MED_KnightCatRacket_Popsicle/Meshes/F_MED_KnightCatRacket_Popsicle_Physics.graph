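# Physics asset for the Popsicle KnightCatRacket medium female body cosmetic, covering its shirt flaps and leg straps for secondary motion.
# Capsules on the pelvis and thighs form the body core, with thighs locked to pelvis.
# Small base anchor bodies lock to the pelvis for shirt pieces and to each thigh for straps.
# Each dyn body hangs from its anchor with per-piece swing angles, tightest on front side flaps, and shirt twist is locked.
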PhysicsAsset /BRCosmetics/Characters/Player/Female/Medium/Bodies/F_MED_KnightCatRacket_Popsicle/Meshes/F_MED_KnightCatRacket_Popsicle_Physics
bodies {
 base_leg_strap_l [{shapeType=Capsule, length=0.505, radius=0.505, shapeName=base_leg_strap_l_capsule, rotation=<rot 0 -90 0>}]
 base_leg_strap_r [{shapeType=Capsule, length=0.505, radius=0.505, shapeName=base_leg_strap_r_capsule, rotation=<rot 0 -90 0>}]
 base_shirt_end_l [{shapeType=Capsule, length=0.5, radius=0.5, shapeName=base_shirt_end_l_capsule}]
 base_shirt_end_r [{shapeType=Capsule, length=0.5, radius=0.5, shapeName=base_shirt_end_r_capsule}]
 base_shirt_leg_back_l [{shapeType=Capsule, length=0.5, radius=0.5, shapeName=base_shirt_leg_back_l_capsule, rotation=<rot -90 0 0>}]
 base_shirt_leg_back_r [{shapeType=Capsule, length=0.5, radius=0.5, shapeName=base_shirt_leg_back_r_capsule, rotation=<rot -90 0 0>}]
 base_shirt_leg_front_M [{shapeType=Capsule, length=0.505, radius=0.505, shapeName=base_shirt_leg_front_M_capsule, rotation=<rot 0 -90 0>}]
 base_shirt_leg_front_l [{shapeType=Capsule, length=0.5, radius=0.5, shapeName=base_shirt_leg_front_l_capsule, rotation=<rot -90 0 0>}]
 base_shirt_leg_front_r [{shapeType=Capsule, length=0.5, radius=0.5, shapeName=base_shirt_leg_front_r_capsule, rotation=<rot -90 0 0>}]
 dyn_leg_strap_l [{shapeType=Capsule, length=1.9864, radius=1.082, shapeName=dyn_leg_strap_l_capsule, center=<2.3709 0 0>, rotation=<rot 90 0 0>}]
 dyn_leg_strap_r [{shapeType=Capsule, length=1.9507, radius=1.2195, shapeName=dyn_leg_strap_r_capsule, center=<-2.2756 0 0>, rotation=<rot 90 0 0>}]
 dyn_shirt_end_l [{shapeType=Capsule, length=5.1576, radius=2, shapeName=dyn_shirt_end_l_capsule, center=<-10 0 -3>, rotation=<rot 90 0 0>}]
 dyn_shirt_end_r [{shapeType=Capsule, length=5.1576, radius=2, shapeName=dyn_shirt_end_r_capsule, center=<10 0 3.1854>, rotation=<rot 90 0 0>}]
 dyn_shirt_leg_back_l [{shapeType=Capsule, length=3.5, radius=3.8, shapeName=dyn_shirt_leg_back_l_capsule, center=<-8 -4.5 0>, rotation=<rot 90 0 0>}]
 dyn_shirt_leg_back_r [{shapeType=Capsule, length=3.5, radius=3.8, shapeName=dyn_shirt_leg_back_r_capsule, center=<8 5 0>, rotation=<rot 90 0 0>}]
 dyn_shirt_leg_front_l [{shapeType=Capsule, length=2, radius=3, shapeName=dyn_shirt_leg_front_l_capsule, center=<-12 1 0>, rotation=<rot 90 0 0>}]
 dyn_shirt_leg_front_m [{shapeType=Capsule, length=2, radius=2, shapeName=dyn_shirt_leg_front_m_capsule, center=<8 0 0>, rotation=<rot 90 0 0>}]
 dyn_shirt_leg_front_r [{shapeType=Capsule, length=2, radius=3, shapeName=dyn_shirt_leg_front_r_capsule, center=<12 -1 0>, rotation=<rot 90 0 0>}]
 pelvis [{shapeType=Capsule, length=13.1669, radius=10, shapeName=pelvis_capsule, center=<-3.4785 -2.1556 0>, rotation=<rot 0 90 0>}]
 thigh_l [{
   shapeType=Capsule
   length=26.8679
   radius=6.5
   shapeName=thigh_l_capsule
   center=<-19.2443 0.2254 0.5996>
   rotation=<rot -0.4302 -88.6275 -92.1632>
  }]
 thigh_r [{
   shapeType=Capsule
   length=26.8367
   radius=6.5
   shapeName=thigh_l_capsule
   center=<19.2337 -0.2278 -0.3187>
   rotation=<rot -0.4302 -88.6275 87.8368>
  }]
}
constraints [
 {
  bone1Name=base_leg_strap_r
  bone2Name=thigh_r
  swing1LimitDegrees=45
  swing1Motion=Locked
  swing2LimitDegrees=45
  swing2Motion=Locked
  twistLimitDegrees=45
  twistMotion=Locked
 }
 {bone1Name=dyn_leg_strap_r, bone2Name=base_leg_strap_r, swing1LimitDegrees=20, swing2LimitDegrees=20, twistLimitDegrees=0}
 {
  bone1Name=base_leg_strap_l
  bone2Name=thigh_l
  swing1LimitDegrees=45
  swing1Motion=Locked
  swing2LimitDegrees=45
  swing2Motion=Locked
  twistLimitDegrees=45
  twistMotion=Locked
 }
 {bone1Name=dyn_leg_strap_l, bone2Name=base_leg_strap_l, swing1LimitDegrees=20, swing2LimitDegrees=20, twistLimitDegrees=0}
 {
  bone1Name=thigh_r
  bone2Name=pelvis
  swing1LimitDegrees=45
  swing1Motion=Locked
  swing2LimitDegrees=45
  swing2Motion=Locked
  twistLimitDegrees=45
  twistMotion=Locked
 }
 {
  bone1Name=thigh_l
  bone2Name=pelvis
  swing1LimitDegrees=45
  swing1Motion=Locked
  swing2LimitDegrees=45
  swing2Motion=Locked
  twistLimitDegrees=45
  twistMotion=Locked
 }
 {
  bone1Name=base_shirt_end_l
  bone2Name=pelvis
  swing1LimitDegrees=45
  swing1Motion=Locked
  swing2LimitDegrees=45
  swing2Motion=Locked
  twistLimitDegrees=45
  twistMotion=Locked
 }
 {
  bone1Name=base_shirt_end_r
  bone2Name=pelvis
  swing1LimitDegrees=45
  swing1Motion=Locked
  swing2LimitDegrees=45
  swing2Motion=Locked
  twistLimitDegrees=45
  twistMotion=Locked
 }
 {bone1Name=dyn_shirt_end_l, bone2Name=base_shirt_end_l, swing1LimitDegrees=20, swing2LimitDegrees=20, twistLimitDegrees=0, twistMotion=Locked}
 {bone1Name=dyn_shirt_end_r, bone2Name=base_shirt_end_r, swing1LimitDegrees=20, swing2LimitDegrees=10, twistLimitDegrees=0, twistMotion=Locked}
 {
  bone1Name=base_shirt_leg_back_l
  bone2Name=pelvis
  swing1LimitDegrees=45
  swing1Motion=Locked
  swing2LimitDegrees=45
  swing2Motion=Locked
  twistLimitDegrees=45
  twistMotion=Locked
 }
 {
  bone1Name=base_shirt_leg_back_r
  bone2Name=pelvis
  swing1LimitDegrees=45
  swing1Motion=Locked
  swing2LimitDegrees=45
  swing2Motion=Locked
  twistLimitDegrees=45
  twistMotion=Locked
 }
 {
  bone1Name=base_shirt_leg_front_l
  bone2Name=pelvis
  swing1LimitDegrees=45
  swing1Motion=Locked
  swing2LimitDegrees=45
  swing2Motion=Locked
  twistLimitDegrees=45
  twistMotion=Locked
 }
 {
  bone1Name=base_shirt_leg_front_r
  bone2Name=pelvis
  swing1LimitDegrees=45
  swing1Motion=Locked
  swing2LimitDegrees=45
  swing2Motion=Locked
  twistLimitDegrees=45
  twistMotion=Locked
 }
 {
  bone1Name=dyn_shirt_leg_front_r
  bone2Name=base_shirt_leg_front_r
  swing1LimitDegrees=3
  swing2LimitDegrees=4
  twistLimitDegrees=0
  twistMotion=Locked
 }
 {
  bone1Name=dyn_shirt_leg_back_r
  bone2Name=base_shirt_leg_back_r
  swing1LimitDegrees=35
  swing2LimitDegrees=35
  twistLimitDegrees=1
  twistMotion=Locked
 }
 {
  bone1Name=dyn_shirt_leg_back_l
  bone2Name=base_shirt_leg_back_l
  swing1LimitDegrees=35
  swing2LimitDegrees=35
  twistLimitDegrees=1
  twistMotion=Locked
 }
 {
  bone1Name=dyn_shirt_leg_front_l
  bone2Name=base_shirt_leg_front_l
  swing1LimitDegrees=3
  swing2LimitDegrees=4
  twistLimitDegrees=0
  twistMotion=Locked
 }
 {
  bone1Name=base_shirt_leg_front_M
  bone2Name=pelvis
  swing1LimitDegrees=45
  swing1Motion=Locked
  swing2LimitDegrees=45
  swing2Motion=Locked
  twistLimitDegrees=45
  twistMotion=Locked
 }
 {
  bone1Name=dyn_shirt_leg_front_m
  bone2Name=base_shirt_leg_front_M
  swing1LimitDegrees=10
  swing2LimitDegrees=10
  twistLimitDegrees=1
  twistMotion=Locked
 }
]
