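# Physics asset for the purse on the female medium KnightCatRacket Popsicle character body, giving capsule bodies to the purse's dyn bones and two spine bones.
# A purse chain runs from spine_05 through three bodies, with front and back branches off the second.
# All purse joints lock twist, and the root and branch joints also lock one swing axis.
# The spine_05 to spine_02 joint locks all three angular axes.
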PhysicsAsset /BRCosmetics/Characters/Player/Female/Medium/Bodies/F_MED_KnightCatRacket_Popsicle/Meshes/F_MED_KnightCatRacket_Popsicle_Purse_Physics
bodies {
 dyn_purse_1 [{shapeType=Capsule, length=2, radius=2, shapeName=dyn_purse_1_capsule, rotation=<rot 90 0 0>}]
 dyn_purse_2 [{shapeType=Capsule, length=0.505, radius=0.505, shapeName=dyn_purse_2_capsule, rotation=<rot 0 -90 0>}]
 dyn_purse_3 [{shapeType=Capsule, length=4, radius=4, shapeName=dyn_purse_3_capsule, center=<8 0 0>, rotation=<rot 0 0 90>}]
 dyn_purse_back_2 [{shapeType=Capsule, length=2, radius=2, shapeName=dyn_purse_back_2_capsule, rotation=<rot 90 0 0>}]
 dyn_purse_front_2 [{shapeType=Capsule, length=2, radius=2, shapeName=dyn_purse_front_2_capsule, rotation=<rot 90 0 0>}]
 spine_02 [{shapeType=Capsule, length=9, radius=9, shapeName=spine_02_capsule, center=<0.4961 -0.2742 0.7245>, rotation=<rot 19.4953 -90 -2.2214>}]
 spine_05 [{shapeType=Capsule, length=8.8842, radius=9.1754, shapeName=spine_05_capsule, center=<-4.3705 0.7792 0>, rotation=<rot 0 89.9996 -180>}]
}
constraints [
 {
  bone1Name=dyn_purse_1
  bone2Name=spine_05
  swing1LimitDegrees=2
  swing1Motion=Locked
  swing2LimitDegrees=10
  twistLimitDegrees=1
  twistMotion=Locked
 }
 {bone1Name=dyn_purse_2, bone2Name=dyn_purse_1, swing1LimitDegrees=4, swing2LimitDegrees=10, twistLimitDegrees=1, twistMotion=Locked}
 {bone1Name=dyn_purse_3, bone2Name=dyn_purse_2, swing1LimitDegrees=6, swing2LimitDegrees=20, twistLimitDegrees=1, twistMotion=Locked}
 {
  bone1Name=dyn_purse_back_2
  bone2Name=dyn_purse_2
  swing1LimitDegrees=2
  swing1Motion=Locked
  swing2LimitDegrees=10
  twistLimitDegrees=1
  twistMotion=Locked
 }
 {
  bone1Name=dyn_purse_front_2
  bone2Name=dyn_purse_2
  swing1LimitDegrees=2
  swing1Motion=Locked
  swing2LimitDegrees=10
  twistLimitDegrees=1
  twistMotion=Locked
 }
 {
  bone1Name=spine_05
  bone2Name=spine_02
  swing1LimitDegrees=45
  swing1Motion=Locked
  swing2LimitDegrees=45
  swing2Motion=Locked
  twistLimitDegrees=45
  twistMotion=Locked
 }
]
