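# Post-process style AnimBlueprint for the female KnightCatRacket Popsicle body cosmetic, animating its shirt, purse, straps and armband.
# It copies the parent mesh pose through deform and shirt Control Rigs, leg-strap constraints and three RigidBody simulations.
# Cached stages of additive bone offsets then correct clipping, weighted by movement, weapon and bone-delta alphas from the EventGraph.
# The EventGraph also sets gravity overrides and feeds WPO cloth-sim material parameters, including skydive wind.
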AnimBlueprint /BRCosmetics/Characters/Player/Female/Medium/Bodies/F_MED_KnightCatRacket_Popsicle/Meshes/F_MED_KnightCatRacket_Popsicle_AnimBP
target_skeleton=&/F_MED_KnightCatRacket_Popsicle_Skeleton
graphs [
 AnimGraph {nodes [
   n1 = "Transform (Modify) Bone - Bone: base_strapmover"(
    ComponentPose=Local To Component(LocalPose=Use cached pose 'Purse_Cache'())
    Alpha=Get StrapAvoid() -> (StrapAvoid:double)
    Translation="6,-1,0"
    Rotation="0,0,0"
    Scale="1,1,1"
   ) {
    rotationMode=BMM_Additive
    rotationSpace=BCS_BoneSpace
    translationMode=BMM_Additive
    translationSpace=BCS_BoneSpace
    alphaBoolBlend {blendOption=Linear}
    alphaScaleBiasClamp {interpSpeedIncreasing=4}
    boneToModify {boneName=base_strapmover}
    translation=<6 -1>
   }
   n2 = "Transform (Modify) Bone - Bone: base_armring"(
    ComponentPose=@n1
    Alpha=Get isTargeting() -> (isTargeting:double)
    Translation="1.037,-0.238,0.046"
    Rotation="0.001,-9.511,-0.000"
    Scale="1,1,1"
   ) {
    rotationMode=BMM_Additive
    rotationSpace=BCS_BoneSpace
    translationMode=BMM_Additive
    translationSpace=BCS_BoneSpace
    alphaBoolBlend {blendInTime=0.5, blendOption=Linear, blendOutTime=0.5}
    alphaScaleBiasClamp {bInterpResult=true}
    boneToModify {boneName=base_armring}
    rotation=<rot 0.001 -9.5109 0>
    translation=<1.0371 -0.2377 0.046>
   }
   n3 = "Transform (Modify) Bone - Bone: base_armring"(ComponentPose=@n2, Alpha=Get Skydive() -> (Skydive:double), Translation="1.767,0.013,-0.330", Rotation="0,3,0", Scale="1,1,1") {
    rotationMode=BMM_Additive
    rotationSpace=BCS_BoneSpace
    translationMode=BMM_Additive
    translationSpace=BCS_BoneSpace
    alphaBoolBlend {blendInTime=0.5, blendOption=Linear, blendOutTime=0.5}
    alphaScaleBiasClamp {bInterpResult=true}
    boneToModify {boneName=base_armring}
    rotation {yaw=3}
    translation=<1.767 0.013 -0.33>
   }
   n4 = "Transform (Modify) Bone - Bone: attach_cape"(ComponentPose=@n3, Alpha=Get Skydive(), Translation="0,-5.5,0", Rotation="0,0,0", Scale="1,1,1") {
    rotationSpace=BCS_BoneSpace
    translationMode=BMM_Additive
    translationSpace=BCS_BoneSpace
    alphaBoolBlend {blendOption=Linear}
    alphaScaleBiasClamp {bInterpResult=true}
    boneToModify {boneName=attach_cape}
    translation {y=-5.5}
   }
   n5 = Get Pouch_shift() -> (Pouch_shift:double)
   n6 = "Transform (Modify) Bone - Bone: dyn_purse_3"(ComponentPose=@n4, Alpha=@n5, Translation="-1.744,-6.031,-0.714", Rotation="-11.753,-21.150,-20.377", Scale="1,1,1") {
    rotationMode=BMM_Additive
    rotationSpace=BCS_BoneSpace
    translationMode=BMM_Additive
    translationSpace=BCS_BoneSpace
    alphaBoolBlend {blendOption=Linear}
    alphaScaleBiasClamp {bInterpResult=true}
    boneToModify {boneName=dyn_purse_3}
    rotation=<rot -11.753 -21.15 -20.377>
    translation=<-1.744 -6.031 -0.714>
   }
   n7 = "Transform (Modify) Bone - Bone: base_purse_back_2"(ComponentPose=@n6, Alpha=@n5, Translation="-5.470,1.844,-2.144", Rotation="-17.489,-7.855,-3.165", Scale="1,1,1") {
    rotationMode=BMM_Additive
    rotationSpace=BCS_BoneSpace
    translationMode=BMM_Additive
    translationSpace=BCS_BoneSpace
    alphaBoolBlend {blendOption=Linear}
    alphaScaleBiasClamp {bInterpResult=true}
    boneToModify {boneName=base_purse_back_2}
    rotation=<rot -17.489 -7.855 -3.165>
    translation=<-5.4697 1.844 -2.1441>
   }
   n8 = "Transform (Modify) Bone - Bone: base_purse_front_2"(ComponentPose=@n7, Alpha=@n5, Translation="-0.164,0.726,-0.218", Rotation="27.142,3.114,-5.017", Scale="1,1,1") {
    rotationMode=BMM_Additive
    rotationSpace=BCS_BoneSpace
    translationMode=BMM_Additive
    translationSpace=BCS_BoneSpace
    alphaBoolBlend {blendOption=Linear}
    alphaScaleBiasClamp {bInterpResult=true, interpSpeedDecreasing=2, interpSpeedIncreasing=4}
    boneToModify {boneName=base_purse_front_2}
    rotation=<rot 27.142 3.114 -5.017>
    translation=<-0.164 0.726 -0.218>
   }
   n9 = Get HipTwist() -> (HipTwist:double)
   n10 = "Transform (Modify) Bone - Bone: base_shirt_leg_front_r"(ComponentPose=@n8, Alpha=@n9, Translation="0,7.034,2.985", Rotation="-0.000,-0.000,-28.371", Scale="1,1,1") {
    rotationMode=BMM_Additive
    rotationSpace=BCS_BoneSpace
    translationMode=BMM_Additive
    translationSpace=BCS_BoneSpace
    alphaBoolBlend {blendOption=Linear}
    alphaScaleBiasClamp {bInterpResult=true, interpSpeedDecreasing=2, interpSpeedIncreasing=4}
    boneToModify {boneName=base_shirt_leg_front_r}
    rotation=<rot 0 0 -28.371>
    translation {y=7.034, z=2.985}
   }
   n11 = "Transform (Modify) Bone - Bone: base_shirt_leg_front_l"(ComponentPose=@n10, Alpha=@n9, Translation="-0.000,-4.411,-2.623", Rotation="-0.000,-0.000,-15.655", Scale="1,1,1") {
    rotationMode=BMM_Additive
    rotationSpace=BCS_BoneSpace
    translationMode=BMM_Additive
    translationSpace=BCS_BoneSpace
    alphaBoolBlend {blendOption=Linear}
    alphaScaleBiasClamp {bInterpResult=true, interpSpeedDecreasing=2, interpSpeedIncreasing=4}
    boneToModify {boneName=base_shirt_leg_front_l}
    rotation {roll=-15.655}
    translation {y=-4.411, z=-2.623}
   }
   n12 = "Transform (Modify) Bone - Bone: base_shirt_end_l"(ComponentPose=@n11, Alpha=@n9, Translation="-0.002,-0.042,-13", Rotation="0,0,0", Scale="1,1,1") {
    rotationMode=BMM_Additive
    rotationSpace=BCS_BoneSpace
    translationMode=BMM_Additive
    translationSpace=BCS_BoneSpace
    alphaBoolBlend {blendInTime=0.5, blendOption=Linear, blendOutTime=0.5}
    alphaScaleBiasClamp {bInterpResult=true}
    boneToModify {boneName=base_shirt_end_l}
    translation=<-0.002 -0.042 -13>
   }
   Output Pose(Result=Component To Local(ComponentPose=@n12)) {layerGroup=None}
   n13 = Control Rig(Source=Copy Pose From Mesh() {bUseAttachedParent=true}) {
    alphaBoolBlend {blendOption=Linear}
    controlRigAssetReference {blueprintRigClass=&/F_MED_DeformRig_KinghtCatRoker_Popsicle.F_MED_DeformRig_KinghtCatRoker_Popsicle_C}
   }
   n14 = Control Rig!(Source=@n13, Alpha, IsCrouched:double=Get Crouch() -> (Crouch:double), IsStrafe:double=0, Side_Left:double=Get Left() -> (Left:double)) -> (Pose) {
    alphaBoolBlend {blendOption=Linear}
    controlRigAssetReference {blueprintRigClass=&/F_MED_KnightCatRacket_Popsicle_Shirt_CtrlRig.F_MED_KnightCatRacket_Popsicle_Shirt_CtrlRig_C}
   }
   Save cached pose 'controlRig_Cache'(Pose=@n14) {cacheName=controlRig_Cache}
   n15 = "Constraint - base_leg_strap_r (PP)"!(
    ComponentPose=Local To Component(LocalPose=Use cached pose 'controlRig_Cache'())
    Alpha=1
    ConstraintWeights_0=0.2
    ConstraintWeights_1:float=0.8
   ) -> (Pose) {
    alphaBoolBlend {blendOption=Linear}
    boneToModify {boneName=base_leg_strap_r}
    constraintSetup [{targetBone {boneName=thigh_r}}, {targetBone {boneName=thigh_twist_01_r}}]
    constraintWeights [1,1]
   }
   n16 = "Constraint - base_leg_strap_l (PP)"!(ComponentPose=@n15, Alpha=1, ConstraintWeights_0=0.2, ConstraintWeights_1:float=0.8) -> (Pose) {
    alphaBoolBlend {blendOption=Linear}
    boneToModify {boneName=base_leg_strap_l}
    constraintSetup [{targetBone {boneName=thigh_l}}, {targetBone {boneName=thigh_twist_01_l}}]
    constraintWeights [1,1]
   }
   n17 = Get GravityOverride()
   n18 = RigidBody!(ComponentPose=@n16, Alpha, OverrideWorldGravity:Vector=@n17, ExternalForce) -> (Pose) {
    bClampLinearTranslationLimitToRefPose=true
    bOverrideWorldGravity=true
    overridePhysicsAsset=&/F_MED_KnightCatRacket_Popsicle_Physics
    simulationSpace=BaseBoneSpace
    alphaBoolBlend {blendOption=Linear}
    baseBoneRef {boneName=pelvis}
    componentAppliedLinearAccClamp=<30000 30000 30000>
    componentLinearAccScale=<0.35 0.2 0.2>
    componentLinearVelScale=<0.35 0.2 0.2>
    overrideWorldGravity {z=-980}
    simSpaceSettings {worldAlpha=0.1}
   }
   n19 = RigidBody(ComponentPose=@n18, OverrideWorldGravity=@n17) {
    bClampLinearTranslationLimitToRefPose=true
    bOverrideWorldGravity=true
    overridePhysicsAsset=&/F_MED_KnightCatRacket_Popsicle_Purse_Physics
    simulationSpace=BaseBoneSpace
    alphaBoolBlend {blendOption=Linear}
    baseBoneRef {boneName=spine_05}
    componentAppliedLinearAccClamp=<6000 6000 6000>
    componentLinearAccScale=<0.4 0.3 0.3>
    componentLinearVelScale=<0.4 0.3 0.3>
    overrideWorldGravity {z=-980}
    simSpaceSettings {worldAlpha=0.25}
   }
   n20 = RigidBody(ComponentPose=@n19, OverrideWorldGravity=Get Armband_Gravity() -> (Armband_Gravity:Vector)) {
    bClampLinearTranslationLimitToRefPose=true
    bOverrideWorldGravity=true
    overridePhysicsAsset=&/F_MED_KnightCatRacket_Popsicle_Armband_Physics
    simulationSpace=BaseBoneSpace
    alphaBoolBlend {blendOption=Linear}
    baseBoneRef {boneName=lowerarm_twist_01_r}
    componentAppliedLinearAccClamp=<3000 3000 3000>
    componentLinearAccScale=<0.3 0.4 0.3>
    componentLinearVelScale=<0.3 0.4 0.3>
    overrideWorldGravity {z=-980}
    simSpaceSettings {
     maxAngularAcceleration=1000
     maxAngularVelocity=3
     maxLinearAcceleration=700
     maxLinearVelocity=500
     velocityScaleZ=0
     worldAlpha=0.2
     externalLinearDragV=<0.2 0.2>
    }
   }
   Save cached pose 'RigidBodies_cache'(Pose=Component To Local(ComponentPose=@n20)) {cacheName=RigidBodies_cache}
   n21 = Get Skydive()
   n22 = "Transform (Modify) Bone - Bone: base_shirt_leg_front_r"(
    ComponentPose=Local To Component(LocalPose=Use cached pose 'RigidBodies_cache'())
    Alpha=@n21
    Translation="-0.000,7.034,2.985"
    Rotation="-0.000,-0.000,-28.371"
    Scale="1,1,1"
   ) {
    rotationMode=BMM_Additive
    rotationSpace=BCS_BoneSpace
    translationMode=BMM_Additive
    translationSpace=BCS_BoneSpace
    alphaBoolBlend {blendOption=Linear}
    alphaScaleBiasClamp {bInterpResult=true, interpSpeedDecreasing=2, interpSpeedIncreasing=4}
    boneToModify {boneName=base_shirt_leg_front_r}
    rotation=<rot 0 0 -28.371>
    translation {y=7.034, z=2.985}
   }
   n23 = "Transform (Modify) Bone - Bone: base_shirt_leg_front_l"(ComponentPose=@n22, Alpha=@n21, Translation="-0.000,-4.411,-2.623", Rotation="-0.000,-0.000,-15.655", Scale="1,1,1") {
    rotationMode=BMM_Additive
    rotationSpace=BCS_BoneSpace
    translationMode=BMM_Additive
    translationSpace=BCS_BoneSpace
    alphaBoolBlend {blendOption=Linear}
    alphaScaleBiasClamp {bInterpResult=true, interpSpeedDecreasing=2, interpSpeedIncreasing=4}
    boneToModify {boneName=base_shirt_leg_front_l}
    rotation {roll=-15.655}
    translation {y=-4.411, z=-2.623}
   }
   n24 = Get Skydive()
   n25 = "Transform (Modify) Bone - Bone: base_shirt_leg_back_l"(ComponentPose=@n23, Alpha=@n24, Translation="-0.971,-10,-3", Rotation="13.437,-14.419,-3.418", Scale="1,1,1") {
    rotationMode=BMM_Additive
    rotationSpace=BCS_BoneSpace
    translationMode=BMM_Additive
    translationSpace=BCS_BoneSpace
    alphaBoolBlend {blendInTime=0.5, blendOption=Linear, blendOutTime=0.5}
    alphaScaleBiasClamp {bInterpResult=true, interpSpeedIncreasing=20}
    boneToModify {boneName=base_shirt_leg_back_l}
    rotation=<rot 13.437 -14.419 -3.418>
    translation=<-0.971 -10 -3>
   }
   n26 = "Transform (Modify) Bone - Bone: base_shirt_leg_back_r"(ComponentPose=@n25, Alpha=@n24, Translation="1.46,10,3", Rotation="0.000,-24.438,0.000", Scale="1,1,1") {
    rotationMode=BMM_Additive
    rotationSpace=BCS_BoneSpace
    translationMode=BMM_Additive
    translationSpace=BCS_BoneSpace
    alphaBoolBlend {blendInTime=0.5, blendOption=Linear, blendOutTime=0.5}
    alphaScaleBiasClamp {bInterpResult=true, interpSpeedIncreasing=20}
    boneToModify {boneName=base_shirt_leg_back_r}
    rotation {roll=0, yaw=-24.438}
    translation=<1.46 10 3>
   }
   n27 = Get Skydive()
   n28 = "Transform (Modify) Bone - Bone: base_shirt_end_l"(ComponentPose=@n26, Alpha=@n27, Translation="0,0,-5", Rotation="0,0,0", Scale="1,1,1") {
    rotationMode=BMM_Additive
    rotationSpace=BCS_BoneSpace
    translationMode=BMM_Additive
    translationSpace=BCS_BoneSpace
    alphaBoolBlend {blendInTime=0.5, blendOption=Linear, blendOutTime=0.5}
    alphaScaleBiasClamp {bInterpResult=true}
    boneToModify {boneName=base_shirt_end_l}
    translation {z=-5}
   }
   n29 = "Transform (Modify) Bone - Bone: base_shirt_end_r"(ComponentPose=@n28, Alpha=@n27, Translation="0,0,5", Rotation="0,0,0", Scale="1,1,1") {
    rotationMode=BMM_Additive
    rotationSpace=BCS_BoneSpace
    translationMode=BMM_Additive
    translationSpace=BCS_BoneSpace
    alphaBoolBlend {blendInTime=0.5, blendOption=Linear, blendOutTime=0.5}
    alphaScaleBiasClamp {bInterpResult=true}
    boneToModify {boneName=base_shirt_end_r}
    translation {z=5}
   }
   n30 = "Transform (Modify) Bone - Bone: base_shirt_leg_front_M"!(
    ComponentPose=@n29
    bAlphaBoolEnabled:bool=Get bIsKneeSliding()
    Translation="0.000,-5.078,0.000"
    Rotation="-0.000,-70.896,-11.716"
    Scale="1,1,1"
   ) -> (Pose) {
    alphaInputType=Bool
    rotationMode=BMM_Additive
    rotationSpace=BCS_BoneSpace
    translationMode=BMM_Additive
    translationSpace=BCS_BoneSpace
    alphaBoolBlend {blendInTime=0.5, blendOption=Linear, blendOutTime=0.5}
    alphaScaleBiasClamp {bInterpResult=true}
    boneToModify {boneName=base_shirt_leg_front_M}
    rotation {roll=-11.716, yaw=-70.896}
    translation {y=-5.078}
   }
   n31 = "Transform (Modify) Bone - Bone: base_shirt_leg_back_fix_m"!(ComponentPose=@n30, bAlphaBoolEnabled:bool=NOT Boolean(A=Get bRigidBodyEnabled()), Translation="0,-3,0", Rotation="0,0,0", Scale="1,1,1") -> (Pose) {
    alphaInputType=Bool
    rotationMode=BMM_Additive
    rotationSpace=BCS_BoneSpace
    translationMode=BMM_Additive
    translationSpace=BCS_BoneSpace
    alphaBoolBlend {blendInTime=0.5, blendOption=Linear, blendOutTime=0.5}
    alphaScaleBiasClamp {bInterpResult=true}
    boneToModify {boneName=base_shirt_leg_back_fix_m}
    translation {y=-3}
   }
   Save cached pose 'ShirtSkyDive_Cache'(Pose=Component To Local(ComponentPose=@n31)) {cacheName=ShirtSkyDive_Cache}
   n32 = Get isMelee() -> (isMelee:double)
   n33 = "Transform (Modify) Bone - Bone: dyn_shirt_end_l"(
    ComponentPose=Local To Component(LocalPose=Use cached pose 'ShirtSkyDive_Cache'())
    Alpha=@n32
    Translation="-0.486,3.748,-10.456"
    Rotation="0,0,0"
    Scale="1,1,1"
   ) {
    rotationMode=BMM_Additive
    rotationSpace=BCS_BoneSpace
    translationMode=BMM_Additive
    translationSpace=BCS_BoneSpace
    alphaBoolBlend {blendOption=Linear}
    alphaScaleBiasClamp {bClampResult=true, bInterpResult=true, clampMax=0.5, interpSpeedDecreasing=2, interpSpeedIncreasing=5}
    boneToModify {boneName=dyn_shirt_end_l}
    translation=<-0.486 3.748 -10.456>
   }
   n34 = "Transform (Modify) Bone - Bone: base_shirt_end_r"(ComponentPose=@n33, Alpha=@n32, Translation="-0.883,-10.006,1.166", Rotation="-0.000,6.433,0.000", Scale="1,1,1") {
    rotationMode=BMM_Additive
    rotationSpace=BCS_BoneSpace
    translationMode=BMM_Additive
    translationSpace=BCS_BoneSpace
    alphaBoolBlend {blendOption=Linear}
    alphaScaleBiasClamp {bClampResult=true, bInterpResult=true, clampMax=0.5, interpSpeedDecreasing=2, interpSpeedIncreasing=5}
    boneToModify {boneName=base_shirt_end_r}
    rotation=<rot 0 6.433 0>
    translation=<-0.883 -10.006 1.166>
   }
   n35 = "Transform (Modify) Bone - Bone: base_shirt_leg_back_r"(ComponentPose=@n34, Alpha=@n32, Translation="-0.000,-0.000,2.832", Rotation="0,0,0", Scale="1,1,1") {
    rotationMode=BMM_Additive
    rotationSpace=BCS_BoneSpace
    translationMode=BMM_Additive
    translationSpace=BCS_BoneSpace
    alphaBoolBlend {blendOption=Linear}
    alphaScaleBiasClamp {bInterpResult=true, interpSpeedDecreasing=2, interpSpeedIncreasing=5}
    boneToModify {boneName=base_shirt_leg_back_r}
    translation {z=2.832}
   }
   n36 = "Transform (Modify) Bone - Bone: base_shirt_leg_front_l"(ComponentPose=@n35, Alpha=@n32, Translation="0,1,6", Rotation="0,0,0", Scale="1,1,1") {
    rotationMode=BMM_Additive
    rotationSpace=BCS_BoneSpace
    translationMode=BMM_Additive
    translationSpace=BCS_BoneSpace
    alphaBoolBlend {blendOption=Linear}
    alphaScaleBiasClamp {bInterpResult=true, interpSpeedDecreasing=2, interpSpeedIncreasing=5}
    boneToModify {boneName=base_shirt_leg_front_l}
    translation {y=1, z=6}
   }
   Save cached pose 'Side_Shirt_Cache'(Pose=Component To Local(ComponentPose=@n36)) {cacheName=Side_Shirt_Cache}
   n37 = "Transform (Modify) Bone - Bone: base_purse_back_2"(
    ComponentPose=Local To Component(LocalPose=Use cached pose 'Side_Shirt_Cache'())
    Alpha=1
    Translation="-1.553,-0.458,-2.027"
    Rotation="0,0,0"
    Scale="1,1,1"
   ) {
    rotationMode=BMM_Additive
    rotationSpace=BCS_BoneSpace
    translationMode=BMM_Additive
    translationSpace=BCS_BoneSpace
    alphaBoolBlend {blendOption=Linear}
    alphaScaleBiasClamp {bInterpResult=true, interpSpeedDecreasing=5, interpSpeedIncreasing=6}
    boneToModify {boneName=base_purse_back_2}
    translation=<-1.553 -0.458 -2.027>
   }
   n38 = "Transform (Modify) Bone - Bone: dyn_purse_3"(ComponentPose=@n37, Alpha=1, Translation="4.493,-2.102,-0.285", Rotation="0,0,0", Scale="1,1,1") {
    rotationMode=BMM_Additive
    rotationSpace=BCS_BoneSpace
    translationMode=BMM_Additive
    translationSpace=BCS_BoneSpace
    alphaBoolBlend {blendOption=Linear}
    alphaScaleBiasClamp {bInterpResult=true, interpSpeedDecreasing=5, interpSpeedIncreasing=6}
    boneToModify {boneName=dyn_purse_3}
    translation=<4.493 -2.102 -0.285>
   }
   n39 = Get IsJumping() -> (IsJumping:double)
   n40 = "Transform (Modify) Bone - Bone: dyn_purse_back_2"(ComponentPose=@n38, Alpha=@n39, Translation="-0.632,-2.082,-1.923", Rotation="-0.476,0.407,-1.619", Scale="1,1,1") {
    rotationMode=BMM_Additive
    rotationSpace=BCS_BoneSpace
    translationMode=BMM_Additive
    translationSpace=BCS_BoneSpace
    alphaBoolBlend {blendOption=Linear}
    alphaScaleBiasClamp {bInterpResult=true, interpSpeedDecreasing=2, interpSpeedIncreasing=4}
    boneToModify {boneName=dyn_purse_back_2}
    rotation=<rot -0.476 0.407 -1.619>
    translation=<-0.632 -2.082 -1.923>
   }
   n41 = "Transform (Modify) Bone - Bone: dyn_purse_3"(ComponentPose=@n40, Alpha=@n39, Translation="3.151,-1.540,-0.900", Rotation="19.227,-2.121,-3.408", Scale="1,1,1") {
    rotationMode=BMM_Additive
    rotationSpace=BCS_BoneSpace
    translationMode=BMM_Additive
    translationSpace=BCS_BoneSpace
    alphaBoolBlend {blendOption=Linear}
    alphaScaleBiasClamp {bInterpResult=true, interpSpeedDecreasing=5, interpSpeedIncreasing=6}
    boneToModify {boneName=dyn_purse_3}
    rotation=<rot 19.227 -2.121 -3.408>
    translation=<3.151 -1.54 -0.9>
   }
   n42 = Get isShortpack() -> (isShortpack:double)
   n43 = "Transform (Modify) Bone - Bone: dyn_purse_3"(ComponentPose=@n41, Alpha=@n42, Translation="-2.476,-0.461,1.158", Rotation="-2.992,-16.025,-3.945", Scale="1,1,1") {
    rotationMode=BMM_Additive
    rotationSpace=BCS_BoneSpace
    translationMode=BMM_Additive
    translationSpace=BCS_BoneSpace
    alphaBoolBlend {blendOption=Linear}
    alphaScaleBiasClamp {bInterpResult=true, interpSpeedDecreasing=6, interpSpeedIncreasing=6}
    boneToModify {boneName=dyn_purse_3}
    rotation=<rot -2.992 -16.025 -3.945>
    translation=<-2.476 -0.461 1.158>
   }
   n44 = "Transform (Modify) Bone - Bone: base_purse_front_2"(ComponentPose=@n43, Alpha=@n42, Translation="-4.564,-2.341,1.876", Rotation="-14.184,17.589,-7.445", Scale="1,1,1") {
    rotationMode=BMM_Additive
    rotationSpace=BCS_BoneSpace
    translationMode=BMM_Additive
    translationSpace=BCS_BoneSpace
    alphaBoolBlend {blendOption=Linear}
    alphaScaleBiasClamp {bInterpResult=true, interpSpeedDecreasing=2, interpSpeedIncreasing=4}
    boneToModify {boneName=base_purse_front_2}
    rotation=<rot -14.184 17.589 -7.445>
    translation=<-4.564 -2.341 1.876>
   }
   n45 = "Transform (Modify) Bone - Bone: base_purse_back_2"(ComponentPose=@n44, Alpha=@n42, Translation="-1.230,-0.252,0.524", Rotation="2.835,0.042,0.954", Scale="1,1,1") {
    rotationMode=BMM_Additive
    rotationSpace=BCS_BoneSpace
    translationMode=BMM_Additive
    translationSpace=BCS_BoneSpace
    alphaBoolBlend {blendOption=Linear}
    alphaScaleBiasClamp {bInterpResult=true, interpSpeedDecreasing=2, interpSpeedIncreasing=4}
    boneToModify {boneName=base_purse_back_2}
    rotation=<rot 2.835 0.042 0.954>
    translation=<-1.23 -0.252 0.524>
   }
   n46 = "Transform (Modify) Bone - Bone: base_purse_back_2"(
    ComponentPose=@n45
    Alpha=Get IsCrouchmove() -> (IsCrouchmove:double)
    Translation="-4.409,0.895,0.820"
    Rotation="3.970,-2.419,-9.109"
    Scale="1,1,1"
   ) {
    rotationMode=BMM_Additive
    rotationSpace=BCS_BoneSpace
    translationMode=BMM_Additive
    translationSpace=BCS_BoneSpace
    alphaBoolBlend {blendOption=Linear}
    alphaScaleBiasClamp {bInterpResult=true, interpSpeedDecreasing=2, interpSpeedIncreasing=4}
    boneToModify {boneName=base_purse_back_2}
    rotation=<rot 3.97 -2.419 -9.109>
    translation=<-4.409 0.8952 0.8201>
   }
   n47 = Get isRun() -> (isRun:double)
   n48 = "Transform (Modify) Bone - Bone: dyn_purse_3"(ComponentPose=@n46, Alpha=@n47, Translation="1.450,-1.290,-1.150", Rotation="-9.366,-6.015,-4.250", Scale="1,1,1") {
    rotationMode=BMM_Additive
    rotationSpace=BCS_BoneSpace
    translationMode=BMM_Additive
    translationSpace=BCS_BoneSpace
    alphaBoolBlend {blendOption=Linear}
    alphaScaleBiasClamp {bInterpResult=true, interpSpeedDecreasing=5, interpSpeedIncreasing=5}
    boneToModify {boneName=dyn_purse_3}
    rotation=<rot -9.366 -6.015 -4.25>
    translation=<1.45 -1.29 -1.15>
   }
   n49 = "Transform (Modify) Bone - Bone: base_purse_back_2"(ComponentPose=@n48, Alpha=@n47, Translation="1.343,-1.932,-0.754", Rotation="10.753,-0.100,4.068", Scale="1,1,1") {
    rotationMode=BMM_Additive
    rotationSpace=BCS_BoneSpace
    translationMode=BMM_Additive
    translationSpace=BCS_BoneSpace
    alphaBoolBlend {blendOption=Linear}
    alphaScaleBiasClamp {bInterpResult=true, interpSpeedDecreasing=1, interpSpeedIncreasing=1}
    boneToModify {boneName=base_purse_back_2}
    rotation=<rot 10.753 -0.1 4.068>
    translation=<1.343 -1.932 -0.754>
   }
   n50 = Get Spine_left() -> (Spine_left:double)
   n51 = "Transform (Modify) Bone - Bone: dyn_purse_3"(ComponentPose=@n49, Alpha=@n50, Translation="0.485,-2.318,1.571", Rotation="1.184,-0.655,8.458", Scale="1,1,1") {
    rotationMode=BMM_Additive
    rotationSpace=BCS_BoneSpace
    translationMode=BMM_Additive
    translationSpace=BCS_BoneSpace
    alphaBoolBlend {blendOption=Linear}
    alphaScaleBiasClamp {bInterpResult=true}
    boneToModify {boneName=dyn_purse_3}
    rotation=<rot 1.184 -0.655 8.458>
    translation=<0.485 -2.3177 1.5709>
   }
   n52 = "Transform (Modify) Bone - Bone: base_purse_back_2"(ComponentPose=@n51, Alpha=@n50, Translation="0.000,-2.802,2.208", Rotation="0,0,0", Scale="1,1,1") {
    rotationMode=BMM_Additive
    rotationSpace=BCS_BoneSpace
    translationMode=BMM_Additive
    translationSpace=BCS_BoneSpace
    alphaBoolBlend {blendOption=Linear}
    alphaScaleBiasClamp {bInterpResult=true}
    boneToModify {boneName=base_purse_back_2}
    translation {y=-2.802, z=2.208}
   }
   n53 = Get Spine_X_Target() -> (Spine_X_Target:double)
   n54 = "Transform (Modify) Bone - Bone: dyn_purse_3"(ComponentPose=@n52, Alpha=@n53, Translation="-2,-1,0", Rotation="0,0,6", Scale="1,1,1") {
    rotationMode=BMM_Additive
    rotationSpace=BCS_BoneSpace
    translationMode=BMM_Additive
    translationSpace=BCS_BoneSpace
    alphaBoolBlend {blendOption=Linear}
    alphaScaleBiasClamp {bInterpResult=true, interpSpeedDecreasing=5, interpSpeedIncreasing=5}
    boneToModify {boneName=dyn_purse_3}
    rotation {roll=6}
    translation=<-2 -1>
   }
   n55 = "Transform (Modify) Bone - Bone: base_purse_back_2"(ComponentPose=@n54, Alpha=@n53, Translation="0,-2,0", Rotation="0,0,0", Scale="1,1,1") {
    rotationMode=BMM_Additive
    rotationSpace=BCS_BoneSpace
    translationMode=BMM_Additive
    translationSpace=BCS_BoneSpace
    alphaBoolBlend {blendOption=Linear}
    alphaScaleBiasClamp {bInterpResult=true}
    boneToModify {boneName=base_purse_back_2}
    translation {y=-2}
   }
   n56 = Get bIsInFrontEnd()
   n57 = "Transform (Modify) Bone - Bone: dyn_purse_3"!(ComponentPose=@n55, bAlphaBoolEnabled:bool=@n56, Translation="0.877,0.249,-0.003", Rotation="0,0,0", Scale="1,1,1") -> (Pose) {
    alphaInputType=Bool
    rotationMode=BMM_Additive
    rotationSpace=BCS_BoneSpace
    translationMode=BMM_Additive
    translationSpace=BCS_BoneSpace
    alphaBoolBlend {blendOption=Linear}
    alphaScaleBiasClamp {bInterpResult=true, interpSpeedDecreasing=2, interpSpeedIncreasing=4}
    boneToModify {boneName=dyn_purse_3}
    translation=<0.877 0.249 -0.003>
   }
   n58 = "Transform (Modify) Bone - Bone: base_purse_front_2"!(ComponentPose=@n57, bAlphaBoolEnabled:bool=@n56, Translation="0.287,0.165,-1.480", Rotation="20.000,0.008,-0.038", Scale="1,1,1") -> (Pose) {
    alphaInputType=Bool
    rotationMode=BMM_Additive
    rotationSpace=BCS_BoneSpace
    translationMode=BMM_Additive
    translationSpace=BCS_BoneSpace
    alphaBoolBlend {blendOption=Linear}
    alphaScaleBiasClamp {bInterpResult=true, interpSpeedDecreasing=2, interpSpeedIncreasing=4}
    boneToModify {boneName=base_purse_front_2}
    rotation=<rot 20 0.008 -0.038>
    translation=<0.287 0.165 -1.48>
   }
   n59 = Get Skydive()
   n60 = "Transform (Modify) Bone - Bone: dyn_purse_3"!(ComponentPose=@n58, Alpha:float=@n59, Translation="0.025,0.685,1.920", Rotation="-12.877,-1.260,5.648", Scale="1,1,1") -> (Pose) {
    rotationMode=BMM_Additive
    rotationSpace=BCS_BoneSpace
    translationMode=BMM_Additive
    translationSpace=BCS_BoneSpace
    alphaBoolBlend {blendOption=Linear}
    alphaScaleBiasClamp {bInterpResult=true, interpSpeedDecreasing=2, interpSpeedIncreasing=4}
    boneToModify {boneName=dyn_purse_3}
    rotation=<rot -12.877 -1.26 5.648>
    translation=<0.025 0.685 1.92>
   }
   n61 = "Transform (Modify) Bone - Bone: base_purse_back_2"(ComponentPose=@n60, Alpha=@n59, Translation="0.000,-0.715,3.652", Rotation="-0.000,0.000,19.375", Scale="1,1,1") {
    rotationMode=BMM_Additive
    rotationSpace=BCS_BoneSpace
    translationMode=BMM_Additive
    translationSpace=BCS_BoneSpace
    alphaBoolBlend {blendOption=Linear}
    alphaScaleBiasClamp {bInterpResult=true}
    boneToModify {boneName=base_purse_back_2}
    rotation=<rot 0 0 19.375>
    translation {y=-0.715, z=3.652}
   }
   n62 = "Transform (Modify) Bone - Bone: base_purse_back_2"(ComponentPose=@n61, Alpha=Get isMelee(), Translation="-0.000,-1.346,0.621", Rotation="-0.000,-3.149,-0.000", Scale="1,1,1") {
    rotationMode=BMM_Additive
    rotationSpace=BCS_BoneSpace
    translationMode=BMM_Additive
    translationSpace=BCS_BoneSpace
    alphaBoolBlend {blendOption=Linear}
    alphaScaleBiasClamp {bInterpResult=true}
    boneToModify {boneName=base_purse_back_2}
    rotation {yaw=-3.149}
    translation {y=-1.346, z=0.621}
   }
   n63 = "Transform (Modify) Bone - Bone: base_purse_back_2"(
    ComponentPose=@n62
    Alpha=Get Iskneeslide() -> (Iskneeslide:double)
    Translation="4.163,-5.535,0.848"
    Rotation="13.019,7.788,0.712"
    Scale="1,1,1"
   ) {
    rotationMode=BMM_Additive
    rotationSpace=BCS_BoneSpace
    translationMode=BMM_Additive
    translationSpace=BCS_BoneSpace
    alphaBoolBlend {blendOption=Linear}
    alphaScaleBiasClamp {bInterpResult=true, interpSpeedDecreasing=2, interpSpeedIncreasing=2}
    boneToModify {boneName=base_purse_back_2}
    rotation=<rot 13.019 7.788 0.712>
    translation=<4.163 -5.535 0.848>
   }
   n64 = "Transform (Modify) Bone - Bone: base_strapmover"(ComponentPose=@n63, Alpha=Get Skydive(), Translation="-3,-2,-1", Rotation="0,0,0", Scale="1,1,1") {
    rotationSpace=BCS_BoneSpace
    translationMode=BMM_Additive
    translationSpace=BCS_BoneSpace
    alphaBoolBlend {blendOption=Linear}
    alphaScaleBiasClamp {bInterpResult=true}
    boneToModify {boneName=base_strapmover}
    translation=<-3 -2 -1>
   }
   n65 = "Transform (Modify) Bone - Bone: base_strapmover"(ComponentPose=@n64, Alpha=1, Translation="-1,0.13,-1.113", Rotation="0,0,0", Scale="1,1,1") {
    rotationSpace=BCS_BoneSpace
    translationMode=BMM_Additive
    translationSpace=BCS_BoneSpace
    alphaBoolBlend {blendOption=Linear}
    alphaScaleBiasClamp {bInterpResult=true}
    boneToModify {boneName=base_strapmover}
    translation=<-1 0.13 -1.113>
   }
   n66 = "Transform (Modify) Bone - Bone: base_strapmover"(ComponentPose=@n65, Alpha=Get IsJumping(), Translation="-4,0,0", Rotation="0,0,0", Scale="1,1,1") {
    rotationSpace=BCS_BoneSpace
    translationMode=BMM_Additive
    translationSpace=BCS_BoneSpace
    alphaBoolBlend {blendOption=Linear}
    alphaScaleBiasClamp {bInterpResult=true}
    boneToModify {boneName=base_strapmover}
    translation {x=-4}
   }
   n67 = Get isTargeting()
   n68 = "Transform (Modify) Bone - Bone: Strap_shoulder_r"(ComponentPose=@n66, Alpha=@n67, Translation="0,0,0.5", Rotation="0,0,0", Scale="1,1,1") {
    translationMode=BMM_Additive
    translationSpace=BCS_BoneSpace
    alphaBoolBlend {blendOption=Linear}
    boneToModify {boneName=Strap_shoulder_r}
    translation {z=0.5}
   }
   n69 = "Transform (Modify) Bone - Bone: base_strapmover"(ComponentPose=@n68, Alpha=@n67, Translation="1,0,-2", Rotation="0,0,0", Scale="1,1,1") {
    rotationSpace=BCS_BoneSpace
    translationMode=BMM_Additive
    translationSpace=BCS_BoneSpace
    alphaBoolBlend {blendOption=Linear}
    alphaScaleBiasClamp {bInterpResult=true}
    boneToModify {boneName=base_strapmover}
    translation {x=1, z=-2}
   }
   n70 = "Transform (Modify) Bone - Bone: Strap_shoulder_r"(ComponentPose=@n69, Alpha=1, Translation="0.616,-0.059,0.417", Rotation="0,0,0", Scale="1,1,1") {
    translationMode=BMM_Additive
    translationSpace=BCS_BoneSpace
    alphaBoolBlend {blendOption=Linear}
    boneToModify {boneName=Strap_shoulder_r}
    translation=<0.616 -0.059 0.417>
   }
   Save cached pose 'Purse_Cache'(Pose=Component To Local(ComponentPose=@n70)) {cacheName=Purse_Cache}
  ]}
 EventGraph {nodes [
   n1 = Event BlueprintUpdateAnimation()
   > n2 = Sequence!(execute) -> (then_0, then_1, then_2:Exec, then_3:Exec, then_4:Exec, then_5:Exec)
   Set WPO Globals Function(
    execute=@n2.then_2
    Material=Get Mat Instance Dynamic()
    Overlay Material=Get Overlay Mat Instance Dynamic()
    Skeletal Mesh=Get SkelMesh()
    "Local Noise Transform  Location"=Get Local Noise Transform Location() -> (Local Noise Transform Location:Vector)
    Local Noise Offset=Get Local Noise Offset()
   )
   > Set WPO Regional Function!(
    execute
    Material:Material Instance Dynamic Object Reference=Get Mat Instance Dynamic()
    Overlay Material:Material Instance Dynamic Object Reference=Get Overlay Mat Instance Dynamic()
    Skeletal Mesh:Skeletal Mesh Component Object Reference=Get SkelMesh()
    Velocity Transform:Name=lowerarm_l
    Velocity Interp Speed:double=Get Velocity Interp Speed()
    Collision Transform:Name=upperarm_l
    Collision Transform Axis:EAxis Enum
    Invert Collision Transform Axis:bool=true
    Delta Time X:double=@n1.DeltaTimeX
    Previous Velocity Transform Location:"Vector (by ref)"=Get R1 Velocity Transform Location()
    Previous Velocity Transform Velocity:"Vector (by ref)"=Get R1 Velocity Transform Velocity()
    Region:int=1
   ) -> (then)
   > Set WPO Regional Function(
    Material=Get Mat Instance Dynamic()
    Overlay Material=Get Overlay Mat Instance Dynamic()
    Skeletal Mesh=Get SkelMesh()
    Velocity Transform=lowerarm_r
    Velocity Interp Speed=Get Velocity Interp Speed()
    Collision Transform=upperarm_r
    Invert Collision Transform Axis=false
    Delta Time X=@n1.DeltaTimeX
    Previous Velocity Transform Location=Get R2 Velocity Transform Location()
    Previous Velocity Transform Velocity=Get R2 Velocity Transform Velocity()
    Region=2
   )
   > n3 = Branch(Condition=Get bIsSkyDiving())
   > n4 = Branch(Condition=Get bRigidBodyEnabled())
   > Set WPO Pho Wind Intensity(WPO Pho Wind Intensity=1.7)
   > n5 = Set WPO Pho Turbulence(WPO Pho Turbulence:double=2.5) -> (Output_Get:double)
   n6 = Get Mat Instance Dynamic()
   n7 = Get Overlay Mat Instance Dynamic()
   SetScalarParameterValue(execute=[@n5.then, @n9.then], self=[@n6, @n7], ParameterName=ClothSim_WindStrength, Value=Get WPO Pho Wind Intensity())
   n8 = Get WPO Pho Turbulence() -> (WPO Pho Turbulence:double)
   > SetScalarParameterValue(self=[@n7, @n6], ParameterName=ClothSim_R1_TurbulenceStrength, Value=@n8)
   > SetScalarParameterValue(self=[@n7, @n6], ParameterName=ClothSim_R2_TurbulenceStrength, Value=@n8)
   Set WPO Pho Wind Intensity(execute=[@n3.else, @n4.else], WPO Pho Wind Intensity=0)
   > n9 = Set WPO Pho Turbulence(WPO Pho Turbulence=0)
   n10 = Sequence!(execute=@n2.then_4) -> (then_0, then_1, then_2, then_3, then_4, then_5, then_6:Exec, then_7:Exec)
   n11 = Get bIsCrouching()
   n12 = Get bIsCrouchSprinting()
   n13 = Get bIsDBNO()
   n14 = Get bIsPlayingEmote()
   n15 = Get bIsSliding()
   n16 = Get bIsBaseballSliding()
   > Set isShortpack(isShortpack:double="To Float (Boolean)"(InBool=OR Boolean!(A=@n11, B=true, C:bool=@n12, D:bool=@n13, E:bool=@n15, F:bool=true, G:bool=@n14, H:bool=@n16) -> (ReturnValue))) -> (Output_Get:double)
   > Set isTargeting(isTargeting:double="To Float (Boolean)"(InBool=Get bIsTargeting())) -> (Output_Get:double)
   > Set isMelee(isMelee:double="To Float (Boolean)"(InBool=Get bIsPlayingMeleeAnim())) -> (Output_Get:double)
   n17 = Get bIsPlayingMeleeAnim()
   n18 = Get bIsCrouching()
   > Set isMeleeCrouch(isMeleeCrouch:double="To Float (Boolean)"(InBool=AND Boolean(A=@n17, B=@n18))) -> (Output_Get:double)
   n19 = PlayerMovement(execute=@n10.then_4, AnimInstance=Self-Reference(), XDotProductCoefficient=100, YDotProductCoefficient=100, ZDotProductCoefficient=100)
   n20 = BreakVector(InVec=MakeVector(X=@n19, Y=@n19.YMovement))
   > Set Left(Left:double="To Float (Boolean)"(InBool="float < float"(A=@n20, B=-50))) -> (Output_Get:double)
   > Set Right(Right:double="To Float (Boolean)"(InBool="float > float"(A=@n20, B=50))) -> (Output_Get:double)
   > Set Forward(Forward:double="To Float (Boolean)"(InBool="float > float"(A=@n20.Y, B=40))) -> (Output_Get:double)
   > Set Backward(Backward:double="To Float (Boolean)"(InBool="float < float"(A=@n20.Y, B=-40))) -> (Output_Get:double)
   n21 = NOT Boolean(A=Get bIsSkyDiving())
   n22 = NOT Boolean(A=Get bIsCrouching())
   n23 = NOT Boolean(A=Get bIsCrouchSprinting())
   n24 = Get bIsMoving2D()
   Set isRun(execute=@n10.then_1, isRun:double="To Float (Boolean)"(InBool=AND Boolean!(A=@n24, B=@n22, C:bool=@n23, D:bool=@n21) -> (ReturnValue))) -> (Output_Get:double)
   n25 = Get bIsMoving2D()
   n26 = Get bIsCrouching()
   > Set IsCrouchmove(IsCrouchmove:double="To Float (Boolean)"(InBool=AND Boolean!(A=@n25, B=@n26) -> (ReturnValue))) -> (Output_Get:double)
   n27 = NOT Boolean(A=Get bIsMoving2D())
   n28 = OR Boolean!(A=Get bIsJumping(), B=Get bIsFalling()) -> (ReturnValue)
   > Set IsJumping(IsJumping:double="To Float (Boolean)"(InBool=AND Boolean(A=@n28, B=@n27))) -> (Output_Get:double)
   n29 = OR Boolean(A=Get bIsSkydiveDiveMode(), B=Get bIsSkydiveFloating())
   > Set Skydive(Skydive:double=SelectFloat(A=0.5, B="To Float (Boolean)"(InBool=@n29), bPickA=Get bIsParachuteOpened())) -> (Output_Get:double)
   > Set Iskneeslide(Iskneeslide:double="To Float (Boolean)"(InBool=Get bIsKneeSliding())) -> (Output_Get:double)
   n30 = Get bIsCrouching()
   Set Crouch(
    execute=@n10.then_3
    Crouch:double=FInterpTo(Current=Get Crouch(), Target="To Float (Boolean)"(InBool=@n30), DeltaTime=@n1.DeltaTimeX, InterpSpeed=4)
   ) -> (Output_Get:double)
   n31 = "To Boolean (Integer)"!(InInt:int=Get ar_firingPosition_GC() -> (ar_firingPosition_GC:int=0)) -> (ReturnValue:bool)
   n32 = Get bIsKneeSliding()
   > Set Pouch_shift(Pouch_shift:double="To Float (Boolean)"(InBool=OR Boolean(A=@n31, B=@n32))) -> (Output_Get:double)
   n33 = Sequence!(execute=@n2.then_3) -> (then_0, then_1, then_2, then_3, then_4)
   n34 = OR Boolean(A=AND Boolean(A=Get bIsCrouching(), B=Get bIsMoving2D()), B=Get bIsKneeSliding())
   n35 = GetOwnerSkeletalMeshComponent()
   n36 = GetDeltaTransformFromRefPose(self=@n35, BoneName=thigh_l, BaseName=pelvis)
   n37 = "float <= float"(A=BreakRotator(InRot=BreakTransform(InTransform=@n36).Rotation), B=3) -> (ReturnValue=~)
   Set isStrafe(execute=@n33.then_4, isStrafe:double="To Float (Boolean)"(InBool=AND Boolean(A=@n37, B=@n34))) -> (Output_Get:double)
   n38 = BreakRotator(InRot=BreakTransform(InTransform=GetDeltaTransformFromRefPose(self=@n35, BoneName=spine_04, BaseName=spine_02)).Rotation)
   Set Spine_X_Target(execute=@n33.then_3, Spine_X_Target:double="To Float (Boolean)"(InBool="float <= float"(A=@n38, B=3) -> (ReturnValue=~))) -> (Output_Get:double)
   n39 = Get bIsPlayingMeleeAnim()
   n40 = Get bIsCrouching()
   Branch(execute=@n33.then_1, Condition=NOT Boolean(A=AND Boolean(A=@n39, B=@n40)))
   > Set HipTwist(HipTwist:double=GetAlphaFromReferenceTransformDelta(
     SkeletalMeshComponent=Get SkelMesh()
     BoneName=pelvis
     BaseName=spine_05
     DeltaComponentType=EulerX
     InputRange=(X=8,Y=100)
     TwistAxis="1,0,0"
    )) -> (Output_Get:double)
   n41 = BreakRotator(InRot=BreakTransform(InTransform=GetDeltaTransformFromRefPose(self=@n35, BoneName=spine_04, BaseName=spine_02)).Rotation)
   Set Spine_left(execute=@n33.then_2, Spine_left:double="To Float (Boolean)"(InBool="float <= float"(A=@n41.Yaw, B=-12) -> (ReturnValue=~))) -> (Output_Get:double)
   Set StrapAvoid(
    execute=@n33.then_0
    StrapAvoid:double=GetAlphaFromReferenceTransformDelta(
     SkeletalMeshComponent=Get SkelMesh()
     BoneName=upperarm_r
     BaseName=spine_05
     DeltaComponentType=EulerY
     InputRange=(X=-40,Y=0)
     TwistAxis="1,0,0"
    )
   ) -> (Output_Get:double)
   n42 = Gravity Override State Function DEPRECATED(
    execute=@n2.then_1
    CharacterPartAnimInstance=Self-Reference()
    GameTimeInSeconds=GetGameTimeInSeconds()
    GravityOverrideParams=Get Gravity Override Params()
   )
   > n43 = Set GravityOverride(GravityOverride=@n42)
   > n44 = Branch(Condition=Get bIsTargeting())
   > Set Armband_Gravity(Armband_Gravity:Vector="0,0,-1500") -> (Output_Get:Vector)
   Set Armband_Gravity(execute=@n44.else, Armband_Gravity=@n43)
   n45 = Cast To FortPlayerAnimInstance(execute=@n2.then_0, Object=GetAnimInstance(self=GetOwnerSkeletalMeshComponent()))
   > Set ar_firingPosition_GC(ar_firingPosition_GC:int=0) -> (Output_Get:int=0)
   > n46 = Switch on EFortWeaponCoreAnimation!(execute, Selection:EFortWeaponCoreAnimation Enum=Get WeaponCoreAnim(self=@n45)) -> (
    Melee:Exec
    Pistol:Exec
    Shotgun:Exec
    PaperBlueprint:Exec
    Rifle:Exec
    MeleeOneHand:Exec
    MachinePistol:Exec
    RocketLauncher:Exec
    GrenadeLauncher:Exec
    GoingCommando:Exec
    AssaultRifle:Exec
    TacticalShotgun:Exec
    SniperRifle:Exec
    TrapPlacement:Exec
    ShoulderLauncher:Exec
    AbilityDecoTool:Exec
    Crossbow:Exec
    C4:Exec
    RemoteControl:Exec
    DualWield:Exec
    AR_BullPup:Exec
    AR_ForwardGrip:Exec
    MedPackPaddles:Exec
    SMG_P90:Exec
    AR_DrumGun:Exec
    Consumable_Small:Exec
    Consumable_Large:Exec
    Balloon:Exec
    MountedTurret:Exec
    CreativeTool:Exec
    ExplosiveBow:Exec
    AshtonIndigo:Exec
    AshtonChicago:Exec
    MeleeDualWield:Exec
    Unarmed:Exec
   )
   n47 = Branch(execute=@n46.GoingCommando)
   > Set ar_firingPosition_GC(ar_firingPosition_GC=1) -> (Output_Get=0)
   Set ar_firingPosition_GC(execute=@n47.else, ar_firingPosition_GC=0) -> (Output_Get=0)
   Event OnCharacterCustomizationLoaded()
   > n48 = Set SkelMesh(SkelMesh=GetPartSkeletalMeshComponent())
   > n49 = Get Dynamic Material Instance(SkeletalMeshComponent=@n48)
   > Set Mat Instance Dynamic(Mat Instance Dynamic=@n49)
   > n50 = GetOrCreateDynamicOverlayMaterial(MeshComponent=@n48)
   > Set Overlay Mat Instance Dynamic(Overlay Mat Instance Dynamic=@n50)
   SetScalarParameterValue(self=[@n7, @n6], ParameterName=ClothSim_R3_TurbulenceStrength, Value=@n8)
   Get bIsKneeSliding()
  ]}
]
variables [
 ar_firingPosition_GC {type=integer}
 Armband_Gravity {type=Vector, default=<0 0 -2000>}
 Backward {type=number}
 Crouch {type=number}
 Forward {type=number}
 Gravity Override Params {
  type=GravityOverrideParamsStruct
  default {
   onGround_Falling {gravityOverride=<0 0 -2000>}
   onGround_Moving {gravityOverride=<0 0 -2000>}
   onGround_Standing {gravityOverride=<0 0 -2000>}
   skydive_DivingDown {gravityOverride=<-1500 0 -2000>}
   skydive_FloatingDown {gravityOverride=<0 -700 -2000>}
   skydive_GoingUp {gravityOverride=<0 -300 -2000>}
   skydive_Parachuting {gravityOverride=<0 -800 -2000>}
  }
 }
 GravityOverride {type=Vector}
 HipTwist {type=number}
 IsCrouchmove {type=number}
 IsJumping {type=number}
 Iskneeslide {type=number}
 isMelee {type=number}
 isMeleeCrouch {type=number}
 isRun {type=number}
 isShortpack {type=number}
 isStrafe {type=number}
 isTargeting {type=number}
 Left {type=number}
 Local Noise Offset {type=Vector}
 Local Noise Transform Location {type=Vector}
 Mat Instance Dynamic {type=MaterialInstanceDynamic}
 Overlay Mat Instance Dynamic {type=MaterialInstanceDynamic}
 Pouch_shift {type=number}
 R1 Velocity Transform Location {type=Vector}
 R1 Velocity Transform Velocity {type=Vector}
 R2 Velocity Transform Location {type=Vector}
 R2 Velocity Transform Velocity {type=Vector}
 "R3  Velocity Transform Location" {type=Vector}
 R3 Velocity Transform Velocity {type=Vector}
 Right {type=number}
 SkelMesh {type=SkeletalMeshComponent}
 Skydive {type=number}
 Spine_left {type=number}
 Spine_X_Target {type=number}
 StrapAvoid {type=number}
 Velocity Interp Speed {type=number}
 WPO Pho Turbulence {type=number}
 WPO Pho Wind Intensity {type=number}
]
