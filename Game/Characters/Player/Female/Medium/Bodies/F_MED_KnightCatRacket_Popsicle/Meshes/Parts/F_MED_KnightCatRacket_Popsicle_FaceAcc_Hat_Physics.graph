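# Physics asset for the hat face accessory on the female medium KnightCatRacket Popsicle player character, setting up secondary motion for the hat's parts.
# A larger head capsule and small capsules on each hat bone form a constraint chain, with base_hat angularly locked to head and dyn_hat to base_hat.
# Front, back and four side bones hang off dyn_hat, locking twist and swing1 so each bends only on swing2, with per-bone angle limits.
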PhysicsAsset /Game/Characters/Player/Female/Medium/Bodies/F_MED_KnightCatRacket_Popsicle/Meshes/Parts/F_MED_KnightCatRacket_Popsicle_FaceAcc_Hat_Physics
bodies {
 base_hat [{shapeType=Capsule, length=1, radius=1, shapeName=base_hat_capsule, center=<1.7295 0 0>, rotation=<rot 90 0 0>}]
 dyn_hat [{shapeType=Capsule, length=1, radius=1, shapeName=dyn_hat_capsule, center=<1.7295 0 0>, rotation=<rot 90 0 0>}]
 dyn_hat_L_side01 [{shapeType=Capsule, length=1, radius=1, shapeName=dyn_hat_L_side01_capsule, center=<0.1394 0 0>, rotation=<rot 90 0 0>}]
 dyn_hat_L_side02 [{shapeType=Capsule, length=1, radius=1, shapeName=dyn_hat_L_side02_capsule, center=<0.1394 0 0>, rotation=<rot 90 0 0>}]
 dyn_hat_R_side01 [{shapeType=Capsule, length=1, radius=1, shapeName=dyn_hat_R_side01_capsule, center=<0.1394 0 0>, rotation=<rot 90 0 0>}]
 dyn_hat_R_side02 [{shapeType=Capsule, length=1, radius=1, shapeName=dyn_hat_R_side02_capsule, center=<0.1394 0 0>, rotation=<rot 90 0 0>}]
 dyn_hat_back [{shapeType=Capsule, length=1, radius=1, shapeName=dyn_hat_back_capsule, center=<0.1394 0 0>, rotation=<rot 90 0 0>}]
 dyn_hat_front [{shapeType=Capsule, length=1, radius=1, shapeName=dyn_hat_front_capsule, center=<0.1394 0 0>, rotation=<rot 90 0 0>}]
 head [{shapeType=Capsule, length=2.75, radius=7, shapeName=head_capsule, center=<4.3692 1.047 -0.4179>, rotation=<rot 90 0 0>}]
}
constraints [
 {
  bone1Name=dyn_hat
  bone2Name=base_hat
  swing1LimitDegrees=5
  swing1Motion=Locked
  swing2LimitDegrees=15
  swing2Motion=Locked
  twistLimitDegrees=1
  twistMotion=Locked
 }
 {
  bone1Name=base_hat
  bone2Name=head
  swing1LimitDegrees=45
  swing1Motion=Locked
  swing2LimitDegrees=45
  swing2Motion=Locked
  twistLimitDegrees=45
  twistMotion=Locked
 }
 {
  bone1Name=dyn_hat_front
  bone2Name=dyn_hat
  swing1LimitDegrees=5
  swing1Motion=Locked
  swing2LimitDegrees=40
  twistLimitDegrees=1
  twistMotion=Locked
 }
 {
  bone1Name=dyn_hat_L_side01
  bone2Name=dyn_hat
  swing1LimitDegrees=5
  swing1Motion=Locked
  swing2LimitDegrees=40
  twistLimitDegrees=1
  twistMotion=Locked
 }
 {
  bone1Name=dyn_hat_L_side02
  bone2Name=dyn_hat
  swing1LimitDegrees=5
  swing1Motion=Locked
  swing2LimitDegrees=30
  twistLimitDegrees=1
  twistMotion=Locked
 }
 {
  bone1Name=dyn_hat_R_side01
  bone2Name=dyn_hat
  swing1LimitDegrees=5
  swing1Motion=Locked
  swing2LimitDegrees=40
  twistLimitDegrees=1
  twistMotion=Locked
 }
 {
  bone1Name=dyn_hat_R_side02
  bone2Name=dyn_hat
  swing1LimitDegrees=5
  swing1Motion=Locked
  swing2LimitDegrees=25
  twistLimitDegrees=1
  twistMotion=Locked
 }
 {
  bone1Name=dyn_hat_back
  bone2Name=dyn_hat
  swing1LimitDegrees=5
  swing1Motion=Locked
  swing2LimitDegrees=35
  twistLimitDegrees=1
  twistMotion=Locked
 }
]
