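# Physics asset for the face accessory mesh of the female medium KnightCatRacket Popsicle character, covering a braid and hair strands.
# A braid chain from base_braid, forking after dyn_braid06, and seven short hair chains (A–G) from base_hair use per-joint swing angles, mostly with twist locked.
# Both bases lock angularly to the head.
# Capsules on head, neck, spine, clavicles and left upper arm form an angularly locked upper-body proxy, with no right upper arm body.
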PhysicsAsset /Game/Characters/Player/Female/Medium/Bodies/F_MED_KnightCatRacket_Popsicle/Meshes/Parts/F_MED_KnightCatRacket_Popsicle_FaceAcc_Physics
bodies {
 base_braid [{shapeType=Capsule, length=1, radius=1, shapeName=base_braid_capsule}]
 base_hair [{shapeType=Capsule, length=0.505, radius=0.505, shapeName=base_hair_capsule, rotation=<rot 0 -90 0>}]
 clavicle_l [{shapeType=Capsule, length=6, radius=6, shapeName=clavicle_l_capsule, center=<5.4485 -1.2074 -2.137>, rotation=<rot 84.375 -0.0018 -0.0018>}]
 clavicle_r [{shapeType=Capsule, length=6, radius=6, shapeName=clavicle_r_capsule, center=<-5.4485 1.2074 2.137>, rotation=<rot 84.375 -0.0018 -0.0018>}]
 dyn_braid01 [{shapeType=Capsule, length=0.75, radius=1.3114, shapeName=dyn_braid01_capsule, center=<3.7747 0 1.0406>, rotation=<rot 90 0 0>}]
 dyn_braid02 [{shapeType=Capsule, length=2.25, radius=1.8465, shapeName=dyn_braid02_capsule, center=<3.0711 0 1.0406>, rotation=<rot 90 0 0>}]
 dyn_braid03 [{shapeType=Capsule, length=0.5, radius=1.8459, shapeName=dyn_braid03_capsule, center=<4.082 0 0>, rotation=<rot 90 0 0>}]
 dyn_braid04 [{shapeType=Capsule, length=1, radius=1.4427, shapeName=dyn_braid04_capsule, center=<2.9603 0 0>, rotation=<rot 90 0 0>}]
 dyn_braid05 [{shapeType=Capsule, length=1, radius=1.3864, shapeName=dyn_braid05_capsule, center=<2.0154 0 0>, rotation=<rot 90 0 0>}]
 dyn_braid06 [{shapeType=Capsule, length=1.5, radius=1.0377, shapeName=dyn_braid06_capsule, center=<2.9769 0 0>, rotation=<rot 90 0 0>}]
 dyn_braid07 [{shapeType=Capsule, length=2, radius=1, shapeName=dyn_braid07_capsule, rotation=<rot 90 0 0>}]
 dyn_braid08 [{shapeType=Capsule, length=4, radius=2.0459, shapeName=dyn_braid08_capsule, center=<3.9163 0 -0.6211>, rotation=<rot 75.9375 -0.0002 -0.0002>}]
 dyn_hairA_01 [{shapeType=Capsule, length=0.75, radius=0.6, shapeName=dyn_hairA_01_capsule, center=<4.3449 0 0>, rotation=<rot 90 0 0>}]
 dyn_hairA_02 [{shapeType=Capsule, length=4, radius=1, shapeName=dyn_hairA_02_capsule, center=<2.89 -0.4456 -0.418>, rotation=<rot 78.7498 -0.0004 2.8121>}]
 dyn_hairB_01 [{shapeType=Capsule, length=0.75, radius=0.5, shapeName=dyn_hairB_01_capsule, center=<1.4736 0 0>, rotation=<rot 90 0 0>}]
 dyn_hairC_01 [{shapeType=Capsule, length=2, radius=1, shapeName=dyn_hairC_01_capsule, center=<1.948 0 0>, rotation=<rot 90 0 0>}]
 dyn_hairD_01 [{shapeType=Capsule, length=2, radius=1, shapeName=dyn_hairD_01_capsule, center=<2.3643 0 0>, rotation=<rot 90 0 0>}]
 dyn_hairE_01 [{
   shapeType=Capsule
   length=1.5
   radius=1.1
   shapeName=dyn_hairE_01_capsule
   center=<2.3422 0.1332 -0.2176>
   rotation=<rot 75.9373 -0.0005 -22.5001>
  }]
 dyn_hairF_01 [{shapeType=Capsule, length=1.5, radius=1, shapeName=dyn_hairF_01_capsule, center=<1.4736 0 0>, rotation=<rot 90 0 0>}]
 dyn_hairF_02 [{shapeType=Capsule, length=1.5, radius=1, shapeName=dyn_hairF_02_capsule, center=<1.4736 0 0>, rotation=<rot 90 0 0>}]
 dyn_hairG_01 [{shapeType=Capsule, length=3.25, radius=2.25, shapeName=dyn_hairG_01_capsule, center=<-0.0716 0 0>, rotation=<rot 90 0 0>}]
 dyn_hairG_02 [{shapeType=Capsule, length=1.5, radius=1, shapeName=dyn_hairG_02_capsule, center=<1.4736 0 0>, rotation=<rot 90 0 0>}]
 head [{shapeType=Capsule, length=2.75, radius=7, shapeName=head_capsule, center=<4.3692 1.047 -0.4179>, rotation=<rot 90 0 0>}]
 neck_01 [{shapeType=Capsule, length=4.75, radius=4, shapeName=neck_01_capsule, center=<1.4186 0 0>, rotation=<rot 90 0 0>}]
 spine_02 [{shapeType=Capsule, length=13, radius=11.5, shapeName=spine_02_capsule, center=<4.1311 -0.5961 0>}]
 spine_03 [{shapeType=Capsule, length=4.5, radius=13.75, shapeName=spine_03_capsule, center=<3.2859 0.2411 0>}]
 spine_04 [{shapeType=Capsule, length=11, radius=9.5, shapeName=spine_04_capsule, center=<3.2723 1.023 0>}]
 spine_05 [
  {shapeType=Capsule, length=35.5, radius=12.755, shapeName=spine_05_capsule, center=<3.5692 -12.4727 -4.83>}
  {shapeType=Capsule, length=11.5, radius=20.75, shapeName=spine_05_capsule, center=<-0.6626 -22.0154 0>, rotation=<rot 90 0 0>}
 ]
 upperarm_l [{
   shapeType=Capsule
   length=20
   radius=6.75
   shapeName=upperarm_l_capsule
   center=<10.8805 -2.1533 -1.7933>
   rotation=<rot 84.375 -0.0018 -0.0018>
  }]
}
constraints [
 {bone1Name=dyn_braid01, bone2Name=base_braid, swing1LimitDegrees=30, swing2LimitDegrees=40, twistLimitDegrees=1, twistMotion=Locked}
 {bone1Name=dyn_braid02, bone2Name=dyn_braid01, swing1LimitDegrees=40, swing2LimitDegrees=40, twistLimitDegrees=1, twistMotion=Locked}
 {bone1Name=dyn_braid03, bone2Name=dyn_braid02, swing1LimitDegrees=40, swing2LimitDegrees=30, twistLimitDegrees=1, twistMotion=Locked}
 {bone1Name=dyn_braid04, bone2Name=dyn_braid03, swing1LimitDegrees=40, swing2LimitDegrees=30, twistLimitDegrees=1, twistMotion=Locked}
 {bone1Name=dyn_braid05, bone2Name=dyn_braid04, swing1LimitDegrees=40, swing2LimitDegrees=35, twistLimitDegrees=1, twistMotion=Locked}
 {bone1Name=dyn_braid06, bone2Name=dyn_braid05, swing1LimitDegrees=60, swing2LimitDegrees=50, twistLimitDegrees=1, twistMotion=Locked}
 {bone1Name=dyn_hairA_02, bone2Name=dyn_hairA_01, swing1LimitDegrees=65, swing2LimitDegrees=30, twistLimitDegrees=1, twistMotion=Locked}
 {bone1Name=dyn_hairF_02, bone2Name=dyn_hairF_01, swing1LimitDegrees=20, swing2LimitDegrees=15, twistLimitDegrees=1, twistMotion=Locked}
 {bone1Name=dyn_hairG_02, bone2Name=dyn_hairG_01, swing1LimitDegrees=20, swing2LimitDegrees=10, twistLimitDegrees=1, twistMotion=Locked}
 {bone1Name=dyn_hairA_01, bone2Name=base_hair, swing1LimitDegrees=50, swing2LimitDegrees=40, twistLimitDegrees=1, twistMotion=Locked}
 {bone1Name=dyn_hairB_01, bone2Name=base_hair, swing1LimitDegrees=20, swing2LimitDegrees=15, twistLimitDegrees=1, twistMotion=Locked}
 {bone1Name=dyn_hairC_01, bone2Name=base_hair, swing1LimitDegrees=15, swing2LimitDegrees=35, twistLimitDegrees=1, twistMotion=Locked}
 {bone1Name=dyn_hairD_01, bone2Name=base_hair, swing1LimitDegrees=25, swing2LimitDegrees=30, twistLimitDegrees=1, twistMotion=Locked}
 {bone1Name=dyn_hairE_01, bone2Name=base_hair, swing1LimitDegrees=25, swing2LimitDegrees=30, twistLimitDegrees=1, twistMotion=Locked}
 {bone1Name=dyn_hairF_01, bone2Name=base_hair, swing1LimitDegrees=20, swing2LimitDegrees=20, twistLimitDegrees=1, twistMotion=Locked}
 {bone1Name=dyn_hairG_01, bone2Name=base_hair, swing1LimitDegrees=10, swing2LimitDegrees=10, twistLimitDegrees=1, twistMotion=Locked}
 {
  bone1Name=base_braid
  bone2Name=head
  swing1LimitDegrees=45
  swing1Motion=Locked
  swing2LimitDegrees=45
  swing2Motion=Locked
  twistLimitDegrees=45
  twistMotion=Locked
 }
 {
  bone1Name=base_hair
  bone2Name=head
  swing1LimitDegrees=45
  swing1Motion=Locked
  swing2LimitDegrees=45
  swing2Motion=Locked
  twistLimitDegrees=45
  twistMotion=Locked
 }
 {
  bone1Name=spine_05
  bone2Name=spine_04
  swing1LimitDegrees=45
  swing1Motion=Locked
  swing2LimitDegrees=45
  swing2Motion=Locked
  twistLimitDegrees=45
  twistMotion=Locked
 }
 {
  bone1Name=spine_04
  bone2Name=spine_03
  swing1LimitDegrees=45
  swing1Motion=Locked
  swing2LimitDegrees=45
  swing2Motion=Locked
  twistLimitDegrees=45
  twistMotion=Locked
 }
 {
  bone1Name=spine_03
  bone2Name=spine_02
  swing1LimitDegrees=45
  swing1Motion=Locked
  swing2LimitDegrees=45
  swing2Motion=Locked
  twistLimitDegrees=45
  twistMotion=Locked
 }
 {
  bone1Name=neck_01
  bone2Name=spine_05
  swing1LimitDegrees=45
  swing1Motion=Locked
  swing2LimitDegrees=45
  swing2Motion=Locked
  twistLimitDegrees=45
  twistMotion=Locked
 }
 {
  bone1Name=clavicle_l
  bone2Name=spine_05
  swing1LimitDegrees=45
  swing1Motion=Locked
  swing2LimitDegrees=45
  swing2Motion=Locked
  twistLimitDegrees=45
  twistMotion=Locked
 }
 {
  bone1Name=clavicle_r
  bone2Name=spine_05
  swing1LimitDegrees=45
  swing1Motion=Locked
  swing2LimitDegrees=45
  swing2Motion=Locked
  twistLimitDegrees=45
  twistMotion=Locked
 }
 {
  bone1Name=upperarm_l
  bone2Name=clavicle_l
  swing1LimitDegrees=45
  swing1Motion=Locked
  swing2LimitDegrees=45
  swing2Motion=Locked
  twistLimitDegrees=45
  twistMotion=Locked
 }
 {
  bone1Name=head
  bone2Name=neck_01
  swing1LimitDegrees=45
  swing1Motion=Locked
  swing2LimitDegrees=45
  swing2Motion=Locked
  twistLimitDegrees=45
  twistMotion=Locked
 }
 {bone1Name=dyn_braid07, bone2Name=dyn_braid06, swing1LimitDegrees=20, swing2LimitDegrees=20, twistLimitDegrees=1}
 {bone1Name=dyn_braid08, bone2Name=dyn_braid06, swing1LimitDegrees=20, swing2LimitDegrees=20, twistLimitDegrees=1}
]
