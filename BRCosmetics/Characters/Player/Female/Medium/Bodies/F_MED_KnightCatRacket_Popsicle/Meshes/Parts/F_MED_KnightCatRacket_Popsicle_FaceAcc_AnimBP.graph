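# Anim Blueprint for the face accessory of the KnightCatRacket Popsicle character, animating braid and hat.
# It copies parent and body mesh poses, then chains RigidBody physics with variants for skydiving, jumping and default, plus bone corrections for DBNO and skydive floating with backbling.
# Look At nodes aim the braid at spine bones when rigid bodies are disabled.
# The EventGraph detects backbling and lowers hat physics alpha during tactical sprint.
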AnimBlueprint /BRCosmetics/Characters/Player/Female/Medium/Bodies/F_MED_KnightCatRacket_Popsicle/Meshes/Parts/F_MED_KnightCatRacket_Popsicle_FaceAcc_AnimBP
target_skeleton=/Game/Characters/Player/Female/Medium/Bodies/F_MED_KnightCatRacket_Popsicle/Meshes/Parts/F_MED_KnightCatRacket_Popsicle_FaceAcc_Skeleton
graphs [
 AnimGraph {nodes [
   n1 = RigidBody!(ComponentPose=Local To Component(LocalPose=Use cached pose 'StartPose'()), bAlphaBoolEnabled:bool, OverrideWorldGravity:Vector="0,0,980") -> (Pose) {
    alphaInputType=Bool
    bClampLinearTranslationLimitToRefPose=true
    bOverrideWorldGravity=true
    overridePhysicsAsset=/Game/Characters/Player/Female/Medium/Bodies/F_MED_KnightCatRacket_Popsicle/Meshes/Parts/F_MED_KnightCatRacket_Popsicle_FaceAcc_Physics
    simulationSpace=BaseBoneSpace
    alphaBoolBlend {blendInTime=0.25, blendOption=Linear, blendOutTime=0.25}
    baseBoneRef {boneName=spine_05}
    componentLinearAccScale=<0.2 0.2 0.2>
    componentLinearVelScale=<0.5 0.5 0.5>
    overrideWorldGravity {z=980}
    simSpaceSettings {maxAngularVelocity=100, maxLinearVelocity=100}
   }
   n2 = Get RB alpha() -> (RB alpha:double)
   n3 = RigidBody!(ComponentPose=@n1, Alpha:float=@n2, OverrideWorldGravity="0,0,400") -> (Pose) {
    bClampLinearTranslationLimitToRefPose=true
    bOverrideWorldGravity=true
    overridePhysicsAsset=/Game/Characters/Player/Female/Medium/Bodies/F_MED_KnightCatRacket_Popsicle/Meshes/Parts/F_MED_KnightCatRacket_Popsicle_FaceAcc_Hat_Physics
    simulationSpace=BaseBoneSpace
    alphaBoolBlend {blendInTime=0.25, blendOption=Linear, blendOutTime=0.25}
    baseBoneRef {boneName=spine_05}
    componentLinearAccScale=<0.2 0.2 0.2>
    componentLinearVelScale=<0.5 0.5 0.5>
    overrideWorldGravity {z=980}
    simSpaceSettings {maxAngularVelocity=100, maxLinearVelocity=100}
   }
   n4 = "Transform (Modify) Bone - Bone: base_braid"!(ComponentPose=@n3, bAlphaBoolEnabled:bool=Get bIsDBNO(), Rotation="-63.857,-51.935,133.589") -> (Pose) {
    alphaInputType=Bool
    rotationMode=BMM_Replace
    alphaBoolBlend {blendOption=Linear}
    boneToModify {boneName=base_braid}
    rotation=<rot -63.857 -51.935 133.589>
   }
   n5 = RigidBody!(ComponentPose=Local To Component(LocalPose=Use cached pose 'StartPose'()), bAlphaBoolEnabled:bool, OverrideWorldGravity="0,0,980") -> (Pose) {
    alphaInputType=Bool
    bClampLinearTranslationLimitToRefPose=true
    bOverrideWorldGravity=true
    overridePhysicsAsset=&F_MED_KnightCatRacket_Popsicle_FaceAcc_Physics
    simulationSpace=BaseBoneSpace
    alphaBoolBlend {blendInTime=0.25, blendOption=Linear, blendOutTime=0.25}
    baseBoneRef {boneName=spine_05}
    componentLinearAccScale=<0.2 0.2 0.2>
    componentLinearVelScale=<0.5 0.5 0.5>
    overrideWorldGravity {z=980}
    simSpaceSettings {maxAngularVelocity=100, maxLinearVelocity=100}
   }
   n6 = RigidBody!(ComponentPose=@n5, Alpha:float=@n2, OverrideWorldGravity="0,0,900") -> (Pose) {
    bClampLinearTranslationLimitToRefPose=true
    bOverrideWorldGravity=true
    overridePhysicsAsset=&F_MED_KnightCatRacket_Popsicle_FaceAcc_Hat_Physics
    simulationSpace=BaseBoneSpace
    alphaBoolBlend {blendInTime=0.25, blendOption=Linear, blendOutTime=0.25}
    baseBoneRef {boneName=spine_05}
    componentLinearAccScale=<0.2 0.2 0.2>
    componentLinearVelScale=<0.5 0.5 0.5>
    overrideWorldGravity {z=980}
    simSpaceSettings {maxAngularVelocity=100, maxLinearVelocity=100}
   }
   n7 = Blend Poses by bool(
    BlendPose_0=Component To Local(ComponentPose=@n6)
    BlendPose_1=Component To Local(ComponentPose=@n4)
    BlendTime_0=0.2
    BlendTime_1=0.6
    bActiveValue=Get bIsJumping()
   ) {blendType=Linear}
   n8 = RigidBody!(ComponentPose=Local To Component(LocalPose=Use cached pose 'StartPose'()), bAlphaBoolEnabled:bool, OverrideWorldGravity="0,0,980") -> (Pose) {
    alphaInputType=Bool
    bClampLinearTranslationLimitToRefPose=true
    bOverrideWorldGravity=true
    overridePhysicsAsset=/Game/Characters/Player/Female/Medium/Bodies/F_MED_KnightCatRacket_Popsicle/Meshes/Parts/F_MED_KnightCatRacket_Popsicle_FaceAcc_Sky_Physics
    simulationSpace=BaseBoneSpace
    alphaBoolBlend {blendInTime=0.25, blendOption=Linear, blendOutTime=0.25}
    baseBoneRef {boneName=spine_05}
    componentLinearAccScale=<0.2 0.2 0.2>
    componentLinearVelScale=<0.5 0.5 0.5>
    overrideWorldGravity {z=980}
    simSpaceSettings {maxAngularVelocity=100, maxLinearVelocity=100}
   }
   n9 = RigidBody!(ComponentPose=@n8, Alpha:float, OverrideWorldGravity="0,0,400") -> (Pose) {
    bClampLinearTranslationLimitToRefPose=true
    bOverrideWorldGravity=true
    overridePhysicsAsset=&F_MED_KnightCatRacket_Popsicle_FaceAcc_Hat_Physics
    simulationSpace=BaseBoneSpace
    alphaBoolBlend {blendInTime=0.25, blendOption=Linear, blendOutTime=0.25}
    baseBoneRef {boneName=spine_05}
    componentLinearAccScale=<0.2 0.2 0.2>
    componentLinearVelScale=<0.5 0.5 0.5>
    overrideWorldGravity {z=980}
    simSpaceSettings {maxAngularVelocity=100, maxLinearVelocity=100}
   }
   n10 = AND Boolean(A=Get bIsSkydiveFloating(), B=Get isWearingBackbling())
   n11 = "Transform (Modify) Bone - Bone: dyn_hat_back"!(ComponentPose=@n9, bAlphaBoolEnabled:bool=@n10, Rotation="0,0,-80") -> (Pose) {
    alphaInputType=Bool
    rotationMode=BMM_Additive
    alphaBoolBlend {blendOption=Linear}
    boneToModify {boneName=dyn_hat_back}
    rotation {roll=-80}
   }
   n12 = "Transform (Modify) Bone - Bone: dyn_hat_L_side02"!(ComponentPose=@n11, bAlphaBoolEnabled:bool=@n10, Rotation="-0.000,20.000,-30.000") -> (Pose) {
    alphaInputType=Bool
    rotationMode=BMM_Additive
    alphaBoolBlend {blendOption=Linear}
    boneToModify {boneName=dyn_hat_L_side02}
    rotation {roll=-30, yaw=20}
   }
   n13 = "Transform (Modify) Bone - Bone: dyn_hat_R_side02"!(ComponentPose=@n12, bAlphaBoolEnabled:bool=@n10, Rotation="20,10,-50") -> (Pose) {
    alphaInputType=Bool
    rotationMode=BMM_Additive
    alphaBoolBlend {blendOption=Linear}
    boneToModify {boneName=dyn_hat_R_side02}
    rotation=<rot 20 10 -50>
   }
   n14 = Component To Local(ComponentPose=@n13)
   n15 = Get bIsSkyDiving()
   n16 = Look At - Bone: dyn_braid01(ComponentPose=Local To Component(LocalPose=Use cached pose 'StartPose'()), Alpha=1) {
    alphaBoolBlend {blendOption=Linear}
    boneToModify {boneName=dyn_braid01}
    lookAtLocation=<11.1949 4.1269 -5.4344>
    lookAtTarget {boneReference {boneName=spine_05}}
    lookAt_Axis {axis=<1 0>}
    lookUp_Axis {axis=<0 1>}
   }
   n17 = Look At - Bone: dyn_braid02(ComponentPose=@n16, Alpha=1) {
    alphaBoolBlend {blendOption=Linear}
    boneToModify {boneName=dyn_braid02}
    lookAtLocation=<3.341 10.0552 -5.4344>
    lookAtTarget {boneReference {boneName=spine_05}}
    lookAt_Axis {axis=<1 0>}
    lookUp_Axis {axis=<0 1>}
   }
   n18 = Look At - Bone: dyn_braid03(ComponentPose=@n17, Alpha=1) {
    alphaBoolBlend {blendOption=Linear}
    boneToModify {boneName=dyn_braid03}
    lookAtLocation=<0 11.3826 -5.4344>
    lookAtTarget {boneReference {boneName=spine_05}}
    lookAt_Axis {axis=<1 0>}
    lookUp_Axis {axis=<0 1>}
   }
   n19 = Look At - Bone: dyn_braid04(ComponentPose=@n18, Alpha=1) {
    alphaBoolBlend {blendOption=Linear}
    boneToModify {boneName=dyn_braid04}
    lookAtLocation=<2.4334 13.384 -5.4344>
    lookAtTarget {boneReference {boneName=spine_04}}
    lookAt_Axis {axis=<1 0>}
    lookUp_Axis {axis=<0 1>}
   }
   n20 = Look At - Bone: dyn_braid05(ComponentPose=@n19, Alpha=1) {
    alphaBoolBlend {blendOption=Linear}
    boneToModify {boneName=dyn_braid05}
    lookAtLocation=<0.2439 13.4246 -5.4344>
    lookAtTarget {boneReference {boneName=spine_04}}
    lookAt_Axis {axis=<1 0>}
    lookUp_Axis {axis=<0 1>}
   }
   n21 = Look At - Bone: dyn_braid06(ComponentPose=@n20, Alpha=1) {
    alphaBoolBlend {blendOption=Linear}
    boneToModify {boneName=dyn_braid06}
    lookAtLocation=<0.346 14.3969 -6.1387>
    lookAtTarget {boneReference {boneName=spine_03}}
    lookAt_Axis {axis=<1 0>}
    lookUp_Axis {axis=<0 1>}
   }
   Output Pose(Result=Blend Poses by bool(
     BlendPose_0=Blend Poses by bool(BlendPose_0=@n14, BlendPose_1=@n7, bActiveValue=@n15) {blendType=Linear}
     BlendPose_1=Component To Local(ComponentPose=@n21)
     bActiveValue=Get bRigidBodyEnabled()
    ) {blendType=Linear}) {layerGroup=None}
   n22 = Layered blend per bone(
    BasePose=Copy Pose From Mesh() {bCopyCurves=true, bUseAttachedParent=true}
    BlendPoses_0=Copy Pose From Mesh(SourceMeshComponent=Get Body_Mesh())
   ) {layerSetup [{branchFilters [{boneName=clavicle_l}, {boneName=clavicle_r}, {boneName=upperarm_r}, {boneName=upperarm_l}]}]}
   Save cached pose 'StartPose'(Pose=@n22) {cacheName=StartPose}
  ]}
 EventGraph {nodes [
   Event OnCharacterCustomizationLoaded()
   > n1 = Cast To PlayerPawn_Athena(Object=TryGetPawnOwner())
   > Set Body_Mesh(Body_Mesh=GetSkeletalMeshForPartType(self=@n1))
   Event BlueprintUpdateAnimation()
   > n2 = Sequence!(execute) -> (then_0, then_1, then_2:Exec)
   > n3 = Gravity Override State Function DEPRECATED(CharacterPartAnimInstance=Self-Reference(), GameTimeInSeconds=GetGameTimeInSeconds(), GravityOverrideParams=Get Gravity Override Params())
   > Set GravityOverride(GravityOverride=@n3)
   n4 = Branch(execute=@n2.then_1, Condition=Get bIsTacticalSprinting())
   > Set RB alpha(RB alpha:double=0.3) -> (Output_Get:double)
   Set RB alpha(execute=@n4.else, RB alpha=1)
   n5 = CheckForBackbling()
   Set isWearingBackbling(
    execute=@n2.then_2
    isWearingBackbling=AND Boolean!(A=@n5, B=NOT Boolean(A=@n5.bIsBackblingACape), C:bool=NOT Boolean(A=@n5.bIsBackblingLong)) -> (ReturnValue)
   )
   > Set Is Wearing Cape(Is Wearing Cape=OR Boolean(A=@n5.bIsBackblingACape, B=@n5.bIsBackblingLong))
  ]}
]
variables [
 Body_Mesh {type=SkeletalMeshComponent}
 Gravity Override Params {type=GravityOverrideParamsStruct}
 GravityOverride {type=Vector}
 Is Wearing Cape {type=boolean}
 isWearingBackbling {type=boolean}
 Linear Scale Backpack {type=Vector}
 Linear Scale Cape {type=Vector}
 RB alpha {type=number}
]
